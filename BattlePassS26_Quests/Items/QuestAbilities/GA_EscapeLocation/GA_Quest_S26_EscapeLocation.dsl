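; Season 26 quest ability that reports when the player escapes a heist base location, via the CustomVerb.EscapedLocation tag.
; On activation it loops over the heist base volume tags and waits for removal of each one the owning pawn has.
; Removal broadcasts a custom verb message carrying the pawn's tags, then sends a gameplay event.
; An author note marks the multi-tag check as a testing workaround; CheckForNearbyEnemies is an empty stub.
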(event Ability|EventActivateAbility
  (bind _owningplayerpawn (Variables|GAVars|Data|GetOwningPlayerPawn))
  ;[1 1 1] Check if player already has the urgent quest, then grant if they dont (don't double grant quest)
  (|CallParentFunction)
  ;;[0.4323 0 0] Workaround to unblock 26.10 testing - this should be fixed to only check against Athena.Location.HeistBase tag once it is fixed and not all of these tags
  (for _array_element (GameplayTags|BreakGameplayTagContainer (Variables|Default|GetHeistbaseVolumeTags))
    (if (GameplayTags|HasMatchingGameplayTag _owningplayerpawn _array_element)
      (Ability|Tasks|WaitGameplayTagRemove _array_element 0 true
        (:Removed
          (bind _result (CustomVerb|AddTagsforSubject (CustomVerb|MakeCustomVerbMessage _owningplayerpawn "(TagName=\"CustomVerb.EscapedLocation\")") "Escape" (GameplayTags|GetOwnedGameplayTags _owningplayerpawn :!class /Script/GameplayTags.GameplayTagAssetInterface)))
          (CustomVerb|BroadcastCustomVerbMessage _result)
          (Ability|SendGameplayEventtoActor 0 "(TagName=\"CustomVerb.EscapedLocation\")"))))))

(event Ability|EventOnEndAbility
  (params (bWasCancelled :type Boolean))
  (|CallParentFunction bWasCancelled))

(event Ability|EventFailedToActivatePassiveAbility
  (params (ActorInfo :type Gameplay%20Ability%20Actor%20Info%20Structure))
  (|CallParentFunction ActorInfo))

(fn CheckForNearbyEnemies
  (params (LastCheck1 :type Boolean)))
